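# Animation Blueprint for the GoggleBud glider cosmetic, animating its ship and two tentacled aliens during skydiving and in frontend levels.
# FrontendLevel selects either the frontend locomotion machine or gameplay locomotion with curve-weighted additive directional leans from skydiving state.
# After an additive jitter, parallel cached chains layer tentacle additives, mostly curve-gated, with variable playback rates in gameplay.
# Bone modifiers then rotate ship plates, shrink Body_Offset by curve and adjust alien head scales.
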AnimBlueprint /BRCosmetics/Animation/Game/MainPlayer/Skydive/ParaGlide/GoggleBud/GoggleBud_AnimBP
target_skeleton=/BRCosmetics/Gadgets/Assets/VinderTech_GliderChute/Glider_GoggleBud/Meshes/GoggleBud_Glider_Skeleton
graphs [
 AnimGraph {nodes [
   n1 = Use cached pose 'Component cache'()
   n2 = Use cached pose 'Component cache FE'()
   n3 = Get FrontendLevel() -> (FrontendLevel:bool)
   Output Pose(Result=Slot 'DefaultSlot'(Source=Blend Poses by bool(BlendPose_0=@n2, BlendPose_1=@n1, bActiveValue=@n3))) {layerGroup=None}
   n4 = Apply Additive!(
    Base=Use cached pose 'Local cache'()
    Additive="Sequence Player (additive) 'GoggleBud_AlienA_TentaculeL_GLIDER'"!(PlayRate:float=Get AlienA_TentaculesRate() -> (AlienA_TentaculesRate:float)) -> (Pose) {method=Graph, sequence=&/GoggleBud_AlienA_TentaculeL_GLIDER}
    AlphaCurveName:Name=alienatentaculel
   ) -> (Pose) {alphaInputType=Curve, alphaBoolBlend {blendOption=Linear}, alphaScaleBiasClamp {bias=1, scale=-1}}
   n5 = "Sequence Player (additive) 'GoggleBud_AlienA_TentaculeR_GLIDER'"!(PlayRate:float=Get AlienA_TentaculesRate()) -> (Pose) {method=Graph, sequence=&/GoggleBud_AlienA_TentaculeR_GLIDER}
   n6 = Get AlienB_TentaculesRate() -> (AlienB_TentaculesRate:float)
   n7 = Apply Additive(
    Base=Apply Additive(
     Base=Apply Additive(Base=@n4, Additive=@n5, AlphaCurveName=alienatentaculer) {alphaInputType=Curve, alphaBoolBlend {blendOption=Linear}, alphaScaleBiasClamp {bias=1, scale=-1}}
     Additive="Sequence Player (additive) 'GoggleBud_AlienB_TentaculeL_GLIDER'"!(PlayRate:float=@n6) -> (Pose) {method=Graph, sequence=&/GoggleBud_AlienB_TentaculeL_GLIDER}
     AlphaCurveName=alienbtentaculel
    ) {alphaInputType=Curve, alphaBoolBlend {blendOption=Linear}, alphaScaleBiasClamp {bias=1, scale=-1}}
    Additive="Sequence Player (additive) 'GoggleBud_AlienB_TentaculeR_GLIDER'"!(PlayRate:float=Get AlienB_TentaculesRate()) -> (Pose) {method=Graph, sequence=&/GoggleBud_AlienB_TentaculeR_GLIDER}
    AlphaCurveName=alienbtentaculer
   ) {alphaInputType=Curve, alphaBoolBlend {blendOption=Linear}, alphaScaleBiasClamp {bias=1, scale=-1}}
   n8 = "Sequence Player (additive) 'GoggleBud_AlienB_TentaculeT_GLIDER'"() {sequence=&/GoggleBud_AlienB_TentaculeT_GLIDER}
   Save cached pose 'TentaculesCache'(Pose=Apply Additive!(
     Base=Apply Additive(Base=@n7, Additive=@n8) {alphaBoolBlend {blendOption=Linear}}
     Additive="Sequence Player (additive) 'GoggleBud_AlienA_TentaculeT_GLIDER'"() {sequence=&/GoggleBud_AlienA_TentaculeT_GLIDER}
     Alpha:float
    ) -> (Pose) {alphaBoolBlend {blendOption=Linear}}) {cacheName=TentaculesCache}
   n9 = "Transform (Modify) Bone - Bone: ship_middelPlate"(
    ComponentPose=Local To Component(LocalPose=Use cached pose 'TentaculesCache_FE'())
    Alpha=1
    Translation="0,0,0"
    Rotation=Get ShipMiddlePlate_Rotator() -> (ShipMiddlePlate_Rotator:Rotator)
    Scale="1,1,1"
   ) {rotationMode=BMM_Replace, rotationSpace=BCS_BoneSpace, alphaBoolBlend {blendOption=Linear}, boneToModify {boneName=ship_middelPlate}}
   n10 = "Transform (Modify) Bone - Bone: ship_outsidePlate"(
    ComponentPose=@n9
    Alpha=1
    Translation="0,0,0"
    Rotation=Get ShipOutsidePlate_Rotator() -> (ShipOutsidePlate_Rotator:Rotator)
    Scale="1,1,1"
   ) {rotationMode=BMM_Replace, rotationSpace=BCS_BoneSpace, alphaBoolBlend {blendOption=Linear}, boneToModify {boneName=ship_outsidePlate}}
   n11 = "Transform (Modify) Bone - Bone: Body_Offset"!(ComponentPose=@n10, AlphaCurveName:Name=Scale, Translation="0,0,0", Rotation="0,0,0", Scale="0.01,0.01,0.01") -> (Pose) {
    alphaInputType=Curve
    scaleMode=BMM_Replace
    alphaBoolBlend {blendOption=Linear}
    boneToModify {boneName=Body_Offset}
    scale=<0.01 0.01 0.01>
   }
   n12 = "Transform (Modify) Bone - Bone: alienA_headBase"(ComponentPose=@n11, Alpha=1, Translation="0,0,0", Rotation="0,0,0", Scale="0.98,1,1") {scaleMode=BMM_Additive, alphaBoolBlend {blendOption=Linear}, boneToModify {boneName=alienA_headBase}, scale {x=0.98}}
   n13 = "Transform (Modify) Bone - Bone: alienA_eye_m"(ComponentPose=@n12, Alpha=1, Translation="0,0,0", Rotation="0,0,0", Scale="0.95,0.95,0.95") {
    scaleMode=BMM_Additive
    scaleSpace=BCS_BoneSpace
    alphaBoolBlend {blendOption=Linear}
    boneToModify {boneName=alienA_eye_m}
    scale=<0.95 0.95 0.95>
   }
   n14 = "Transform (Modify) Bone - Bone: alienA_glassContainer_01"(ComponentPose=@n13, Alpha=1, Translation="0,0,0", Rotation="0,0,0", Scale="0.98,1,1") {scaleMode=BMM_Additive, alphaBoolBlend {blendOption=Linear}, boneToModify {boneName=alienA_glassContainer_01}, scale {x=0.98}}
   n15 = "Transform (Modify) Bone - Bone: alienB_headBase"(ComponentPose=@n14, Alpha=1, Translation="0,0,0", Rotation="0,0,0", Scale="0.98,1,1") {scaleMode=BMM_Additive, alphaBoolBlend {blendOption=Linear}, boneToModify {boneName=alienB_headBase}, scale {x=0.98}}
   n16 = "Transform (Modify) Bone - Bone: alienB_eye_m"(ComponentPose=@n15, Alpha=1, Translation="0,0,0", Rotation="0,0,0", Scale="0.95,0.95,0.95") {scaleMode=BMM_Additive, alphaBoolBlend {blendOption=Linear}, boneToModify {boneName=alienB_eye_m}, scale=<0.95 0.95 0.95>}
   n17 = "Transform (Modify) Bone - Bone: alienB_glassContainer_01"(ComponentPose=@n16, Alpha=1, Translation="0,0,0", Rotation="0,0,0", Scale="0.98,1,1") {scaleMode=BMM_Additive, alphaBoolBlend {blendOption=Linear}, boneToModify {boneName=alienB_glassContainer_01}, scale {x=0.98}}
   Save cached pose 'Component cache FE'(Pose=Component To Local(ComponentPose=@n17)) {cacheName=Component cache FE}
   n18 = "Sequence Player (additive) 'GoggleBud_AlienA_TentaculeL_FrontEnd_GLIDER'"!(PlayRate:float=1) -> (Pose) {sequence=&/GoggleBud_AlienA_TentaculeL_FrontEnd_GLIDER}
   n19 = Use cached pose 'Local cache'()
   n20 = Apply Additive!(
    Base=Apply Additive(
     Base=Apply Additive(Base=@n19, Additive=@n18, AlphaCurveName=alienatentaculel) {alphaInputType=Curve, alphaBoolBlend {blendOption=Linear}, alphaScaleBiasClamp {bias=1, scale=-1}}
     Additive="Sequence Player (additive) 'GoggleBud_AlienA_TentaculeR_FrontEnd_GLIDER'"!(PlayRate:float=1) -> (Pose) {sequence=&/GoggleBud_AlienA_TentaculeR_FrontEnd_GLIDER}
     AlphaCurveName=alienatentaculer
    ) {alphaInputType=Curve, alphaBoolBlend {blendOption=Linear}, alphaScaleBiasClamp {bias=1, scale=-1}}
    Additive="Sequence Player (additive) 'GoggleBud_AlienB_TentaculeL_FrontEnd_GLIDER'"!(PlayRate:float=1) -> (Pose) {sequence=&/GoggleBud_AlienB_TentaculeL_FrontEnd_GLIDER}
    AlphaCurveName:Name=alienbtentaculel
   ) -> (Pose) {alphaInputType=Curve, alphaBoolBlend {blendOption=Linear}, alphaScaleBiasClamp {bias=1, scale=-1}}
   n21 = "Sequence Player (additive) 'GoggleBud_AlienB_TentaculeR_FrontEnd_GLIDER'"!(PlayRate:float=1) -> (Pose) {sequence=&/GoggleBud_AlienB_TentaculeR_FrontEnd_GLIDER}
   Save cached pose 'TentaculesCache_FE'(Pose=Apply Additive!(
     Base=Apply Additive!(Base=@n20, Additive=@n21, AlphaCurveName:Name=alienbtentaculer) -> (Pose) {alphaInputType=Curve, alphaBoolBlend {blendOption=Linear}, alphaScaleBiasClamp {bias=1, scale=-1}}
     Additive="Sequence Player (additive) 'GoggleBud_Aliens_TentaculeT_FrontEnd_GLIDER'"() {sequence=&/GoggleBud_Aliens_TentaculeT_FrontEnd_GLIDER}
     Alpha:float
    ) -> (Pose) {alphaBoolBlend {blendOption=Linear}}) {cacheName=TentaculesCache_FE}
   AnimGraphNode_StateMachine_0 = LeanTransitions() {is=title AnimGraphNode_StateMachine}
   AnimGraphNode_StateMachine_2 = Locomotion() {is=title AnimGraphNode_StateMachine}
   n22 = Apply Additive(Base=@AnimGraphNode_StateMachine_2, Additive=@AnimGraphNode_StateMachine_0, AlphaCurveName=ParaGlide_Lean) {alphaInputType=Curve, alphaBoolBlend {blendOption=Linear}, alphaScaleBiasClamp {bias=1, scale=-1}}
   n23 = Get FrontendLevel()
   AnimGraphNode_StateMachine_1 = Locomotion_Frontend() {is=title AnimGraphNode_StateMachine}
   Save cached pose 'Local cache'(Pose=Apply Additive!(
     Base=Blend Poses by bool(BlendPose_0=@AnimGraphNode_StateMachine_1, BlendPose_1=@n22, bActiveValue=@n23) {blendType=Linear}
     Additive="Sequence Player (additive) 'GoggleBud_Jitter_Anim_GLIDER'"() {sequence=&/GoggleBud_Jitter_Anim_GLIDER}
     Alpha:float
    ) -> (Pose) {alphaBoolBlend {blendOption=Linear}}) {cacheName=Local cache}
   n24 = "Transform (Modify) Bone - Bone: ship_middelPlate"(
    ComponentPose=Local To Component(LocalPose=Use cached pose 'TentaculesCache'())
    Alpha=1
    Translation="0,0,0"
    Rotation=Get ShipMiddlePlate_Rotator()
    Scale="1,1,1"
   ) {rotationMode=BMM_Replace, rotationSpace=BCS_BoneSpace, alphaBoolBlend {blendOption=Linear}, boneToModify {boneName=ship_middelPlate}}
   n25 = "Transform (Modify) Bone - Bone: ship_outsidePlate"(ComponentPose=@n24, Alpha=1, Translation="0,0,0", Rotation=Get ShipOutsidePlate_Rotator(), Scale="1,1,1") {rotationMode=BMM_Replace, rotationSpace=BCS_BoneSpace, alphaBoolBlend {blendOption=Linear}, boneToModify {boneName=ship_outsidePlate}}
   n26 = "Transform (Modify) Bone - Bone: Body_Offset"(ComponentPose=@n25, AlphaCurveName=Scale, Translation="0,0,0", Rotation="0,0,0", Scale="0.01,0.01,0.01") {
    alphaInputType=Curve
    scaleMode=BMM_Replace
    alphaBoolBlend {blendOption=Linear}
    boneToModify {boneName=Body_Offset}
    scale=<0.01 0.01 0.01>
   }
   n27 = "Transform (Modify) Bone - Bone: alienA_headBase"(ComponentPose=@n26, Alpha=1, Translation="0,0,0", Rotation="0,0,0", Scale="0.98,1,1") {scaleMode=BMM_Additive, alphaBoolBlend {blendOption=Linear}, boneToModify {boneName=alienA_headBase}, scale {x=0.98}}
   n28 = "Transform (Modify) Bone - Bone: alienA_eye_m"(ComponentPose=@n27, Alpha=1, Translation="0,0,0", Rotation="0,0,0", Scale="0.95,0.95,0.95") {
    scaleMode=BMM_Additive
    scaleSpace=BCS_BoneSpace
    alphaBoolBlend {blendOption=Linear}
    boneToModify {boneName=alienA_eye_m}
    scale=<0.95 0.95 0.95>
   }
   n29 = "Transform (Modify) Bone - Bone: alienA_glassContainer_01"(ComponentPose=@n28, Alpha=1, Translation="0,0,0", Rotation="0,0,0", Scale="0.98,1,1") {scaleMode=BMM_Additive, alphaBoolBlend {blendOption=Linear}, boneToModify {boneName=alienA_glassContainer_01}, scale {x=0.98}}
   n30 = "Transform (Modify) Bone - Bone: alienB_headBase"(ComponentPose=@n29, Alpha=1, Translation="0,0,0", Rotation="0,0,0", Scale="0.98,1,1") {
    scaleMode=BMM_Additive
    scaleSpace=BCS_BoneSpace
    alphaBoolBlend {blendOption=Linear}
    boneToModify {boneName=alienB_headBase}
    scale {x=0.98}
   }
   n31 = "Transform (Modify) Bone - Bone: alienB_eye_m"(ComponentPose=@n30, Alpha=1, Translation="0,0,0", Rotation="0,0,0", Scale="0.95,0.95,0.95") {scaleMode=BMM_Additive, alphaBoolBlend {blendOption=Linear}, boneToModify {boneName=alienB_eye_m}, scale=<0.95 0.95 0.95>}
   n32 = "Transform (Modify) Bone - Bone: alienB_glassContainer_01"(ComponentPose=@n31, Alpha=1, Translation="0,0,0", Rotation="0,0,0", Scale="0.98,1,1") {scaleMode=BMM_Additive, alphaBoolBlend {blendOption=Linear}, boneToModify {boneName=alienB_glassContainer_01}, scale {x=0.98}}
   Save cached pose 'Component cache'(Pose=Component To Local(ComponentPose=@n32)) {cacheName=Component cache}
  ]}
 LeanTransitions {
  entry=Lean_Zero
  path=AnimGraph.AnimGraphNode_StateMachine_0.LeanTransitions
  ids=id
  nodes [
   AnimStateNode_0 = Lean_Zero() {
    is=title AnimStateNode
    stateEntered {guid=#1}
    stateFullyBlended {guid=#2}
    stateLeft {guid=#3}
   }
   AnimStateNode_1 = Lean() {is=title AnimStateNode}
   AnimStateNode_2 = Lean_Out() {is=title AnimStateNode}
   AnimStateNode_5 = Lean_Into() {
    is=title AnimStateNode
    stateEntered {guid=#4}
    stateFullyBlended {guid=#5}
    stateLeft {guid=#6}
   }
   AnimStateTransitionNode_0 = AnimStateTransitionNode()
   AnimStateTransitionNode_1 = AnimStateTransitionNode() {bAutomaticRuleBasedOnSequencePlayerInState=true, crossfadeDuration=0}
   AnimStateTransitionNode_18 = AnimStateTransitionNode() {bAutomaticRuleBasedOnSequencePlayerInState=true, blendMode=Linear, priorityOrder=2}
   AnimStateTransitionNode_19 = AnimStateTransitionNode() {blendMode=Linear}
   AnimStateTransitionNode_2 = AnimStateTransitionNode() {crossfadeDuration=0.32}
   AnimStateTransitionNode_3 = AnimStateTransitionNode()
   AnimStateTransitionNode_4 = AnimStateTransitionNode()
   AnimStateTransitionNode_5 = AnimStateTransitionNode() {crossfadeDuration=0.32}
  ]
  links [
   {
    dst=Lean_Into
    rule=AnimGraph.AnimGraphNode_StateMachine_0.LeanTransitions.AnimStateTransitionNode_5.Transition
    src=Lean
    via=AnimStateTransitionNode_5
   }
   {
    dst=Lean_Out
    rule=AnimGraph.AnimGraphNode_StateMachine_0.LeanTransitions.AnimStateTransitionNode_2.Transition
    src=Lean
    via=AnimStateTransitionNode_2
   }
   {
    dst=Lean
    rule=AnimGraph.AnimGraphNode_StateMachine_0.LeanTransitions.AnimStateTransitionNode_1.Transition
    src=Lean_Into
    via=AnimStateTransitionNode_1
   }
   {
    dst=Lean_Out
    rule=AnimGraph.AnimGraphNode_StateMachine_0.LeanTransitions.AnimStateTransitionNode_3.Transition
    src=Lean_Into
    via=AnimStateTransitionNode_3
   }
   {
    dst=Lean_Into
    rule=AnimGraph.AnimGraphNode_StateMachine_0.LeanTransitions.AnimStateTransitionNode_4.Transition
    src=Lean_Out
    via=AnimStateTransitionNode_4
   }
   {
    dst=Lean_Zero
    rule=AnimGraph.AnimGraphNode_StateMachine_0.LeanTransitions.AnimStateTransitionNode_18.Transition
    src=Lean_Out
    via=AnimStateTransitionNode_18
   }
   {
    dst=Lean_Zero
    rule=AnimGraph.AnimGraphNode_StateMachine_0.LeanTransitions.AnimStateTransitionNode_19.Transition
    src=Lean_Out
    via=AnimStateTransitionNode_19
   }
   {
    dst=Lean_Into
    rule=AnimGraph.AnimGraphNode_StateMachine_0.LeanTransitions.AnimStateTransitionNode_0.Transition
    src=Lean_Zero
    via=AnimStateTransitionNode_0
   }
  ]
 }
 Lean_Zero {
  path=AnimGraph.AnimGraphNode_StateMachine_0.LeanTransitions.AnimStateNode_0.Lean_Zero
  nodes [Output Animation Pose(Result=Blend Poses by bool(
     BlendPose_0="Sequence Player (additive) 'GoggleBud_Glide_Center_Anim_GLIDER'"() {sequence=&/GoggleBud_Glide_Center_Anim_GLIDER}
     BlendPose_1="Sequence Player (additive) 'GoggleBud_Lean_Center_Pose_GLIDER'"() {sequence=&/GoggleBud_Lean_Center_Pose_GLIDER}
     bActiveValue=Get IntoEnd() -> (IntoEnd:bool)
    ) {blendType=Linear})]
 }
 Lean {
  path=AnimGraph.AnimGraphNode_StateMachine_0.LeanTransitions.AnimStateNode_1.Lean
  nodes [
   AnimGraphNode_StateMachine_0 = Lean() {is=title AnimGraphNode_StateMachine}
   Output Animation Pose(Result=@AnimGraphNode_StateMachine_0)
  ]
 }
 Lean {
  entry=Zero
  path=AnimGraph.AnimGraphNode_StateMachine_0.LeanTransitions.AnimStateNode_1.Lean.AnimGraphNode_StateMachine_0.Lean
  ids=id
  nodes [
   To Lean Left() {is=title AnimStateAliasNode, also {aliases [Lean_Back,Lean_Forward,Lean_Right,Zero]}}
   To Lean Foward() {is=title AnimStateAliasNode, also {aliases [Lean_Back,Lean_Left,Lean_Right,Zero]}}
   To Lean Right() {is=title AnimStateAliasNode, also {aliases [Lean_Back,Lean_Forward,Lean_Left,Zero]}}
   To Lean Back() {is=title AnimStateAliasNode, also {aliases [Lean_Forward,Lean_Left,Lean_Right,Zero]}}
   AnimStateNode_0 = Zero() {is=title AnimStateNode}
   AnimStateNode_1 = Lean_Forward() {is=title AnimStateNode}
   AnimStateNode_2 = Lean_Back() {is=title AnimStateNode}
   AnimStateNode_3 = Lean_Left() {is=title AnimStateNode}
   AnimStateNode_4 = Lean_Right() {is=title AnimStateNode}
   AnimStateTransitionNode_16 = AnimStateTransitionNode() {crossfadeDuration=0.5}
   AnimStateTransitionNode_17 = AnimStateTransitionNode() {crossfadeDuration=0.5}
   AnimStateTransitionNode_18 = AnimStateTransitionNode() {crossfadeDuration=0.5}
   AnimStateTransitionNode_19 = AnimStateTransitionNode() {crossfadeDuration=0.5}
  ]
  links [
   {
    dst=Lean_Back
    rule=AnimGraph.AnimGraphNode_StateMachine_0.LeanTransitions.AnimStateNode_1.Lean.AnimGraphNode_StateMachine_0.Lean.AnimStateTransitionNode_19.AnimationTransitionGraph_0
    src=To Lean Back
    via=AnimStateTransitionNode_19
   }
   {
    dst=Lean_Forward
    rule=AnimGraph.AnimGraphNode_StateMachine_0.LeanTransitions.AnimStateNode_1.Lean.AnimGraphNode_StateMachine_0.Lean.AnimStateTransitionNode_17.AnimationTransitionGraph_0
    src=To Lean Foward
    via=AnimStateTransitionNode_17
   }
   {
    dst=Lean_Left
    rule=AnimGraph.AnimGraphNode_StateMachine_0.LeanTransitions.AnimStateNode_1.Lean.AnimGraphNode_StateMachine_0.Lean.AnimStateTransitionNode_16.AnimationTransitionGraph_0
    src=To Lean Left
    via=AnimStateTransitionNode_16
   }
   {
    dst=Lean_Right
    rule=AnimGraph.AnimGraphNode_StateMachine_0.LeanTransitions.AnimStateNode_1.Lean.AnimGraphNode_StateMachine_0.Lean.AnimStateTransitionNode_18.AnimationTransitionGraph_0
    src=To Lean Right
    via=AnimStateTransitionNode_18
   }
  ]
 }
 Zero {
  path=AnimGraph.AnimGraphNode_StateMachine_0.LeanTransitions.AnimStateNode_1.Lean.AnimGraphNode_StateMachine_0.Lean.AnimStateNode_0.Zero
  nodes [Output Animation Pose(Result="Sequence Player (additive) 'GoggleBud_Lean_Center_Pose_GLIDER'"() {sequence=&GoggleBud_Lean_Center_Pose_GLIDER})]
 }
 Lean_Forward {
  path=AnimGraph.AnimGraphNode_StateMachine_0.LeanTransitions.AnimStateNode_1.Lean.AnimGraphNode_StateMachine_0.Lean.AnimStateNode_1.Lean_Forward
  nodes [Output Animation Pose(Result="Sequence Player (additive) 'GoggleBud_Glide_Forward_Anim_GLIDER'"() {sequence=&/GoggleBud_Glide_Forward_Anim_GLIDER})]
 }
 Lean_Back {
  path=AnimGraph.AnimGraphNode_StateMachine_0.LeanTransitions.AnimStateNode_1.Lean.AnimGraphNode_StateMachine_0.Lean.AnimStateNode_2.Lean_Back
  nodes [Output Animation Pose(Result="Sequence Player (additive) 'GoggleBud_Glide_Back_Anim_GLIDER'"() {sequence=&/GoggleBud_Glide_Back_Anim_GLIDER})]
 }
 Lean_Left {
  path=AnimGraph.AnimGraphNode_StateMachine_0.LeanTransitions.AnimStateNode_1.Lean.AnimGraphNode_StateMachine_0.Lean.AnimStateNode_3.Lean_Left
  nodes [Output Animation Pose(Result="Sequence Player (additive) 'GoggleBud_Glide_Left_Anim_GLIDER'"() {sequence=&/GoggleBud_Glide_Left_Anim_GLIDER})]
 }
 Lean_Right {
  path=AnimGraph.AnimGraphNode_StateMachine_0.LeanTransitions.AnimStateNode_1.Lean.AnimGraphNode_StateMachine_0.Lean.AnimStateNode_4.Lean_Right
  nodes [Output Animation Pose(Result="Sequence Player (additive) 'GoggleBud_Glide_Right_Anim_GLIDER'"() {sequence=&/GoggleBud_Glide_Right_Anim_GLIDER})]
 }
 AnimationTransitionGraph_0 {
  path=AnimGraph.AnimGraphNode_StateMachine_0.LeanTransitions.AnimStateNode_1.Lean.AnimGraphNode_StateMachine_0.Lean.AnimStateTransitionNode_16.AnimationTransitionGraph_0
  nodes [Result(bCanEnterTransition=Break Fort Anim Input Skydiving!(FortAnimInput_Skydiving:"Fort Anim Input Skydiving Structure (by ref)"=Get SkydivingState() -> (SkydivingState:Fort Anim Input Skydiving Structure)) -> (
     ParaGilderRootOffset:Vector=~
     ParaGliderRootRotationOverride:Rotator=~
     bUseParaGlideRootModifier:bool=~
     bIsSkydivingFromLaunchPad:bool=~
     bIsSkydivingFromBus:bool=~
     bIsInVortex:bool=~
     bIsDiving:bool=~
     bIsParachuteOpen:bool=~
     bIsSkydiving:bool=~
     bIsParachuteLeaning:bool=~
     bIsSkydiveLeaning:bool=~
     bIsLeaning:bool=~
     bIsSkydiveDiveMode:bool=~
     bPlaySkydiveDrift:bool=~
     bSkydiveDriftDelayActive:bool=~
     bSkydiveDriftAnimAllowed:bool=~
     bIsGliderRight:bool=~
     bIsGliderCenter:bool=~
     bIsGliderLeft:bool=~
     bIsGliderForward:bool=~
     bIsGliderBack:bool=~
     bHasDeployedAndIsStandingUpright:bool=~
     LocalAccelForward:float=~
     LocalAccelRight:float=~
     LocalVelocityRight:float=~
     SkydiveAimPitch:float=~
     SkydiveAimYaw:float=~
     DeployChuteAnimRate:float=~
     SkydiveDriftAnimRate:float=~
     SkydiveDriftAnimRateCurrent:float=~
     SkydiveFidgetAnimRate:float=~
     SkydiveFidgetAnimRateCurrent:float=~
     SkydiveAdditiveAlpha:float=~
     SkydiveDriftDelay:float=~
     SkydiveDriftAnim:int
     SkydiveDriftAnimMax:int
     LaunchpadAnim:int
     LocalAccelDir:ESkydivingDirection Enum
     DirectionLast:ESkydivingDirection Enum
     PoseAdditive:Anim Sequence Object Reference
    ) {is=title K2Node_BreakStruct}.bIsGliderLeft)]
 }
 AnimationTransitionGraph_0 {
  path=AnimGraph.AnimGraphNode_StateMachine_0.LeanTransitions.AnimStateNode_1.Lean.AnimGraphNode_StateMachine_0.Lean.AnimStateTransitionNode_17.AnimationTransitionGraph_0
  nodes [Result(bCanEnterTransition=Break Fort Anim Input Skydiving(FortAnimInput_Skydiving=Get SkydivingState()) -> (
     ParaGilderRootOffset=~
     ParaGliderRootRotationOverride=~
     bUseParaGlideRootModifier=~
     bIsSkydivingFromLaunchPad=~
     bIsSkydivingFromBus=~
     bIsInVortex=~
     bIsDiving=~
     bIsParachuteOpen=~
     bIsSkydiving=~
     bIsParachuteLeaning=~
     bIsSkydiveLeaning=~
     bIsLeaning=~
     bIsSkydiveDiveMode=~
     bPlaySkydiveDrift=~
     bSkydiveDriftDelayActive=~
     bSkydiveDriftAnimAllowed=~
     bIsGliderRight=~
     bIsGliderCenter=~
     bIsGliderLeft=~
     bIsGliderForward=~
     bIsGliderBack=~
     bHasDeployedAndIsStandingUpright=~
     LocalAccelForward=~
     LocalAccelRight=~
     LocalVelocityRight=~
     SkydiveAimPitch=~
     SkydiveAimYaw=~
     DeployChuteAnimRate=~
     SkydiveDriftAnimRate=~
     SkydiveDriftAnimRateCurrent=~
     SkydiveFidgetAnimRate=~
     SkydiveFidgetAnimRateCurrent=~
     SkydiveAdditiveAlpha=~
     SkydiveDriftDelay=~
    ).bIsGliderForward)]
 }
 AnimationTransitionGraph_0 {
  path=AnimGraph.AnimGraphNode_StateMachine_0.LeanTransitions.AnimStateNode_1.Lean.AnimGraphNode_StateMachine_0.Lean.AnimStateTransitionNode_18.AnimationTransitionGraph_0
  nodes [Result(bCanEnterTransition=Break Fort Anim Input Skydiving(FortAnimInput_Skydiving=Get SkydivingState()) -> (
     ParaGilderRootOffset=~
     ParaGliderRootRotationOverride=~
     bUseParaGlideRootModifier=~
     bIsSkydivingFromLaunchPad=~
     bIsSkydivingFromBus=~
     bIsInVortex=~
     bIsDiving=~
     bIsParachuteOpen=~
     bIsSkydiving=~
     bIsParachuteLeaning=~
     bIsSkydiveLeaning=~
     bIsLeaning=~
     bIsSkydiveDiveMode=~
     bPlaySkydiveDrift=~
     bSkydiveDriftDelayActive=~
     bSkydiveDriftAnimAllowed=~
     bIsGliderRight=~
     bIsGliderCenter=~
     bIsGliderLeft=~
     bIsGliderForward=~
     bIsGliderBack=~
     bHasDeployedAndIsStandingUpright=~
     LocalAccelForward=~
     LocalAccelRight=~
     LocalVelocityRight=~
     SkydiveAimPitch=~
     SkydiveAimYaw=~
     DeployChuteAnimRate=~
     SkydiveDriftAnimRate=~
     SkydiveDriftAnimRateCurrent=~
     SkydiveFidgetAnimRate=~
     SkydiveFidgetAnimRateCurrent=~
     SkydiveAdditiveAlpha=~
     SkydiveDriftDelay=~
    ).bIsGliderRight)]
 }
 AnimationTransitionGraph_0 {
  path=AnimGraph.AnimGraphNode_StateMachine_0.LeanTransitions.AnimStateNode_1.Lean.AnimGraphNode_StateMachine_0.Lean.AnimStateTransitionNode_19.AnimationTransitionGraph_0
  nodes [Result(bCanEnterTransition=Break Fort Anim Input Skydiving(FortAnimInput_Skydiving=Get SkydivingState()) -> (
     ParaGilderRootOffset=~
     ParaGliderRootRotationOverride=~
     bUseParaGlideRootModifier=~
     bIsSkydivingFromLaunchPad=~
     bIsSkydivingFromBus=~
     bIsInVortex=~
     bIsDiving=~
     bIsParachuteOpen=~
     bIsSkydiving=~
     bIsParachuteLeaning=~
     bIsSkydiveLeaning=~
     bIsLeaning=~
     bIsSkydiveDiveMode=~
     bPlaySkydiveDrift=~
     bSkydiveDriftDelayActive=~
     bSkydiveDriftAnimAllowed=~
     bIsGliderRight=~
     bIsGliderCenter=~
     bIsGliderLeft=~
     bIsGliderForward=~
     bIsGliderBack=~
     bHasDeployedAndIsStandingUpright=~
     LocalAccelForward=~
     LocalAccelRight=~
     LocalVelocityRight=~
     SkydiveAimPitch=~
     SkydiveAimYaw=~
     DeployChuteAnimRate=~
     SkydiveDriftAnimRate=~
     SkydiveDriftAnimRateCurrent=~
     SkydiveFidgetAnimRate=~
     SkydiveFidgetAnimRateCurrent=~
     SkydiveAdditiveAlpha=~
     SkydiveDriftDelay=~
    ).bIsGliderBack)]
 }
 Lean_Out {
  path=AnimGraph.AnimGraphNode_StateMachine_0.LeanTransitions.AnimStateNode_2.Lean_Out
  nodes [
   n1 = "Blend Poses (ESkydivingDirection)"!(
    BlendPose_0:Pose Link Structure="Sequence Player (additive) 'GoggleBud_Lean_Center_Pose_GLIDER'"() {sequence=&GoggleBud_Lean_Center_Pose_GLIDER}
    BlendPose_1:Pose Link Structure="Sequence Player (additive) 'GoggleBud_Lean_Forward_Out_GLIDER'"() {sequence=&/GoggleBud_Lean_Forward_Out_GLIDER}
    BlendPose_2:Pose Link Structure="Sequence Player (additive) 'GoggleBud_Lean_Back_Out_GLIDER'"() {sequence=&/GoggleBud_Lean_Back_Out_GLIDER}
    BlendPose_3:Pose Link Structure="Sequence Player (additive) 'GoggleBud_Lean_Left_Out_GLIDER'"() {sequence=&/GoggleBud_Lean_Left_Out_GLIDER}
    BlendPose_4:Pose Link Structure="Sequence Player (additive) 'GoggleBud_Lean_Right_Out_GLIDER'"() {sequence=&/GoggleBud_Lean_Right_Out_GLIDER}
    BlendTime_0:float=0.5
    BlendTime_1:float=0.5
    BlendTime_2:float=0.5
    BlendTime_3:float=0.5
    BlendTime_4:float=0.5
    ActiveEnumValue:Byte=Break Fort Anim Input Skydiving(FortAnimInput_Skydiving=Get SkydivingState()) -> (
     ParaGilderRootOffset=~
     ParaGliderRootRotationOverride=~
     bUseParaGlideRootModifier=~
     bIsSkydivingFromLaunchPad=~
     bIsSkydivingFromBus=~
     bIsInVortex=~
     bIsDiving=~
     bIsParachuteOpen=~
     bIsSkydiving=~
     bIsParachuteLeaning=~
     bIsSkydiveLeaning=~
     bIsLeaning=~
     bIsSkydiveDiveMode=~
     bPlaySkydiveDrift=~
     bSkydiveDriftDelayActive=~
     bSkydiveDriftAnimAllowed=~
     bIsGliderRight=~
     bIsGliderCenter=~
     bIsGliderLeft=~
     bIsGliderForward=~
     bIsGliderBack=~
     bHasDeployedAndIsStandingUpright=~
     LocalAccelForward=~
     LocalAccelRight=~
     LocalVelocityRight=~
     SkydiveAimPitch=~
     SkydiveAimYaw=~
     DeployChuteAnimRate=~
     SkydiveDriftAnimRate=~
     SkydiveDriftAnimRateCurrent=~
     SkydiveFidgetAnimRate=~
     SkydiveFidgetAnimRateCurrent=~
     SkydiveAdditiveAlpha=~
     SkydiveDriftDelay=~
    ).DirectionLast
   ) -> (Pose:Pose Link Structure) {
    is=title AnimGraphNode_BlendListByEnum
    blendType=Linear
    blendPose [
     {linkId=-1, sourceLinkId=-1}
     {linkId=-1, sourceLinkId=-1}
     {linkId=-1, sourceLinkId=-1}
     {linkId=-1, sourceLinkId=-1}
     {linkId=-1, sourceLinkId=-1}
    ]
    blendTime [0.1,0.1,0.1,0.1,0.1]
   }
   Output Animation Pose(Result=@n1)
  ]
 }
 Lean_Into {
  path=AnimGraph.AnimGraphNode_StateMachine_0.LeanTransitions.AnimStateNode_5.Lean_Into
  nodes [
   AnimGraphNode_StateMachine_0 = Lean_Into() {is=title AnimGraphNode_StateMachine}
   Output Animation Pose(Result=@AnimGraphNode_StateMachine_0)
  ]
 }
 Lean_Into {
  entry=Zero
  path=AnimGraph.AnimGraphNode_StateMachine_0.LeanTransitions.AnimStateNode_5.Lean_Into.AnimGraphNode_StateMachine_0.Lean_Into
  ids=id
  nodes [
   To Lean Left Into() {is=title AnimStateAliasNode, also {aliases [Lean_Back_Into, Lean_Forward_Into, Lean_Right_Into, Zero]}}
   To Lean Foward Into() {is=title AnimStateAliasNode, also {aliases [Lean_Back_Into, Lean_Left_Into, Lean_Right_Into, Zero]}}
   To Lean Right Into() {is=title AnimStateAliasNode, also {aliases [Lean_Back_Into, Lean_Forward_Into, Lean_Left_Into, Zero]}}
   To Lean Back Into() {is=title AnimStateAliasNode, also {aliases [Lean_Forward_Into, Lean_Left_Into, Lean_Right_Into, Zero]}}
   AnimStateNode_0 = Zero()
   AnimStateNode_1 = Lean_Forward_Into() {is=title AnimStateNode}
   AnimStateNode_2 = Lean_Back_Into() {is=title AnimStateNode}
   AnimStateNode_3 = Lean_Left_Into() {is=title AnimStateNode}
   AnimStateNode_4 = Lean_Right_Into() {is=title AnimStateNode}
   AnimStateTransitionNode_16 = AnimStateTransitionNode() {crossfadeDuration=0.5}
   AnimStateTransitionNode_17 = AnimStateTransitionNode() {crossfadeDuration=0.5}
   AnimStateTransitionNode_18 = AnimStateTransitionNode() {crossfadeDuration=0.5}
   AnimStateTransitionNode_19 = AnimStateTransitionNode() {crossfadeDuration=0.5}
  ]
  links [
   {
    dst=Lean_Back_Into
    rule=AnimGraph.AnimGraphNode_StateMachine_0.LeanTransitions.AnimStateNode_5.Lean_Into.AnimGraphNode_StateMachine_0.Lean_Into.AnimStateTransitionNode_19.AnimationTransitionGraph_0
    src=To Lean Back Into
    via=AnimStateTransitionNode_19
   }
   {
    dst=Lean_Forward_Into
    rule=AnimGraph.AnimGraphNode_StateMachine_0.LeanTransitions.AnimStateNode_5.Lean_Into.AnimGraphNode_StateMachine_0.Lean_Into.AnimStateTransitionNode_17.AnimationTransitionGraph_0
    src=To Lean Foward Into
    via=AnimStateTransitionNode_17
   }
   {
    dst=Lean_Left_Into
    rule=AnimGraph.AnimGraphNode_StateMachine_0.LeanTransitions.AnimStateNode_5.Lean_Into.AnimGraphNode_StateMachine_0.Lean_Into.AnimStateTransitionNode_16.AnimationTransitionGraph_0
    src=To Lean Left Into
    via=AnimStateTransitionNode_16
   }
   {
    dst=Lean_Right_Into
    rule=AnimGraph.AnimGraphNode_StateMachine_0.LeanTransitions.AnimStateNode_5.Lean_Into.AnimGraphNode_StateMachine_0.Lean_Into.AnimStateTransitionNode_18.AnimationTransitionGraph_0
    src=To Lean Right Into
    via=AnimStateTransitionNode_18
   }
  ]
 }
 Zero {
  path=AnimGraph.AnimGraphNode_StateMachine_0.LeanTransitions.AnimStateNode_5.Lean_Into.AnimGraphNode_StateMachine_0.Lean_Into.AnimStateNode_0.Zero
  nodes [Output Animation Pose(Result="Sequence Player (additive) 'GoggleBud_Lean_Center_Pose_GLIDER'"() {sequence=&GoggleBud_Lean_Center_Pose_GLIDER})]
 }
 Lean_Forward_Into {
  path=AnimGraph.AnimGraphNode_StateMachine_0.LeanTransitions.AnimStateNode_5.Lean_Into.AnimGraphNode_StateMachine_0.Lean_Into.AnimStateNode_1.Lean_Forward_Into
  nodes [
   n1 = "Sequence Player (additive) 'GoggleBud_Lean_Forward_Into_GLIDER'"!(bLoopAnimation:bool=False) -> (Pose) {bLoopAnimation=false, sequence=&/GoggleBud_Lean_Forward_Into_GLIDER}
   Output Animation Pose(Result=@n1)
  ]
 }
 Lean_Back_Into {
  path=AnimGraph.AnimGraphNode_StateMachine_0.LeanTransitions.AnimStateNode_5.Lean_Into.AnimGraphNode_StateMachine_0.Lean_Into.AnimStateNode_2.Lean_Back_Into
  nodes [Output Animation Pose(Result="Sequence Player (additive) 'GoggleBud_Lean_Back_Into_GLIDER'"!(bLoopAnimation:bool=False) -> (Pose) {bLoopAnimation=false, sequence=&/GoggleBud_Lean_Back_Into_GLIDER})]
 }
 Lean_Left_Into {
  path=AnimGraph.AnimGraphNode_StateMachine_0.LeanTransitions.AnimStateNode_5.Lean_Into.AnimGraphNode_StateMachine_0.Lean_Into.AnimStateNode_3.Lean_Left_Into
  nodes [Output Animation Pose(Result="Sequence Player (additive) 'GoggleBud_Lean_Left_Into_GLIDER'"!(bLoopAnimation:bool=False) -> (Pose) {bLoopAnimation=false, sequence=&/GoggleBud_Lean_Left_Into_GLIDER})]
 }
 Lean_Right_Into {
  path=AnimGraph.AnimGraphNode_StateMachine_0.LeanTransitions.AnimStateNode_5.Lean_Into.AnimGraphNode_StateMachine_0.Lean_Into.AnimStateNode_4.Lean_Right_Into
  nodes [Output Animation Pose(Result="Sequence Player (additive) 'GoggleBud_Lean_Right_Into_GLIDER'"!(bLoopAnimation:bool=False) -> (Pose) {bLoopAnimation=false, sequence=&/GoggleBud_Lean_Right_Into_GLIDER})]
 }
 AnimationTransitionGraph_0 {
  path=AnimGraph.AnimGraphNode_StateMachine_0.LeanTransitions.AnimStateNode_5.Lean_Into.AnimGraphNode_StateMachine_0.Lean_Into.AnimStateTransitionNode_16.AnimationTransitionGraph_0
  nodes [Result(bCanEnterTransition=Break Fort Anim Input Skydiving(FortAnimInput_Skydiving=Get SkydivingState()) -> (
     ParaGilderRootOffset=~
     ParaGliderRootRotationOverride=~
     bUseParaGlideRootModifier=~
     bIsSkydivingFromLaunchPad=~
     bIsSkydivingFromBus=~
     bIsInVortex=~
     bIsDiving=~
     bIsParachuteOpen=~
     bIsSkydiving=~
     bIsParachuteLeaning=~
     bIsSkydiveLeaning=~
     bIsLeaning=~
     bIsSkydiveDiveMode=~
     bPlaySkydiveDrift=~
     bSkydiveDriftDelayActive=~
     bSkydiveDriftAnimAllowed=~
     bIsGliderRight=~
     bIsGliderCenter=~
     bIsGliderLeft=~
     bIsGliderForward=~
     bIsGliderBack=~
     bHasDeployedAndIsStandingUpright=~
     LocalAccelForward=~
     LocalAccelRight=~
     LocalVelocityRight=~
     SkydiveAimPitch=~
     SkydiveAimYaw=~
     DeployChuteAnimRate=~
     SkydiveDriftAnimRate=~
     SkydiveDriftAnimRateCurrent=~
     SkydiveFidgetAnimRate=~
     SkydiveFidgetAnimRateCurrent=~
     SkydiveAdditiveAlpha=~
     SkydiveDriftDelay=~
    ).bIsGliderLeft)]
 }
 AnimationTransitionGraph_0 {
  path=AnimGraph.AnimGraphNode_StateMachine_0.LeanTransitions.AnimStateNode_5.Lean_Into.AnimGraphNode_StateMachine_0.Lean_Into.AnimStateTransitionNode_17.AnimationTransitionGraph_0
  nodes [Result(bCanEnterTransition=Break Fort Anim Input Skydiving(FortAnimInput_Skydiving=Get SkydivingState()) -> (
     ParaGilderRootOffset=~
     ParaGliderRootRotationOverride=~
     bUseParaGlideRootModifier=~
     bIsSkydivingFromLaunchPad=~
     bIsSkydivingFromBus=~
     bIsInVortex=~
     bIsDiving=~
     bIsParachuteOpen=~
     bIsSkydiving=~
     bIsParachuteLeaning=~
     bIsSkydiveLeaning=~
     bIsLeaning=~
     bIsSkydiveDiveMode=~
     bPlaySkydiveDrift=~
     bSkydiveDriftDelayActive=~
     bSkydiveDriftAnimAllowed=~
     bIsGliderRight=~
     bIsGliderCenter=~
     bIsGliderLeft=~
     bIsGliderForward=~
     bIsGliderBack=~
     bHasDeployedAndIsStandingUpright=~
     LocalAccelForward=~
     LocalAccelRight=~
     LocalVelocityRight=~
     SkydiveAimPitch=~
     SkydiveAimYaw=~
     DeployChuteAnimRate=~
     SkydiveDriftAnimRate=~
     SkydiveDriftAnimRateCurrent=~
     SkydiveFidgetAnimRate=~
     SkydiveFidgetAnimRateCurrent=~
     SkydiveAdditiveAlpha=~
     SkydiveDriftDelay=~
    ).bIsGliderForward)]
 }
 AnimationTransitionGraph_0 {
  path=AnimGraph.AnimGraphNode_StateMachine_0.LeanTransitions.AnimStateNode_5.Lean_Into.AnimGraphNode_StateMachine_0.Lean_Into.AnimStateTransitionNode_18.AnimationTransitionGraph_0
  nodes [Result(bCanEnterTransition=Break Fort Anim Input Skydiving(FortAnimInput_Skydiving=Get SkydivingState()) -> (
     ParaGilderRootOffset=~
     ParaGliderRootRotationOverride=~
     bUseParaGlideRootModifier=~
     bIsSkydivingFromLaunchPad=~
     bIsSkydivingFromBus=~
     bIsInVortex=~
     bIsDiving=~
     bIsParachuteOpen=~
     bIsSkydiving=~
     bIsParachuteLeaning=~
     bIsSkydiveLeaning=~
     bIsLeaning=~
     bIsSkydiveDiveMode=~
     bPlaySkydiveDrift=~
     bSkydiveDriftDelayActive=~
     bSkydiveDriftAnimAllowed=~
     bIsGliderRight=~
     bIsGliderCenter=~
     bIsGliderLeft=~
     bIsGliderForward=~
     bIsGliderBack=~
     bHasDeployedAndIsStandingUpright=~
     LocalAccelForward=~
     LocalAccelRight=~
     LocalVelocityRight=~
     SkydiveAimPitch=~
     SkydiveAimYaw=~
     DeployChuteAnimRate=~
     SkydiveDriftAnimRate=~
     SkydiveDriftAnimRateCurrent=~
     SkydiveFidgetAnimRate=~
     SkydiveFidgetAnimRateCurrent=~
     SkydiveAdditiveAlpha=~
     SkydiveDriftDelay=~
    ).bIsGliderRight)]
 }
 AnimationTransitionGraph_0 {
  path=AnimGraph.AnimGraphNode_StateMachine_0.LeanTransitions.AnimStateNode_5.Lean_Into.AnimGraphNode_StateMachine_0.Lean_Into.AnimStateTransitionNode_19.AnimationTransitionGraph_0
  nodes [Result(bCanEnterTransition=Break Fort Anim Input Skydiving(FortAnimInput_Skydiving=Get SkydivingState()) -> (
     ParaGilderRootOffset=~
     ParaGliderRootRotationOverride=~
     bUseParaGlideRootModifier=~
     bIsSkydivingFromLaunchPad=~
     bIsSkydivingFromBus=~
     bIsInVortex=~
     bIsDiving=~
     bIsParachuteOpen=~
     bIsSkydiving=~
     bIsParachuteLeaning=~
     bIsSkydiveLeaning=~
     bIsLeaning=~
     bIsSkydiveDiveMode=~
     bPlaySkydiveDrift=~
     bSkydiveDriftDelayActive=~
     bSkydiveDriftAnimAllowed=~
     bIsGliderRight=~
     bIsGliderCenter=~
     bIsGliderLeft=~
     bIsGliderForward=~
     bIsGliderBack=~
     bHasDeployedAndIsStandingUpright=~
     LocalAccelForward=~
     LocalAccelRight=~
     LocalVelocityRight=~
     SkydiveAimPitch=~
     SkydiveAimYaw=~
     DeployChuteAnimRate=~
     SkydiveDriftAnimRate=~
     SkydiveDriftAnimRateCurrent=~
     SkydiveFidgetAnimRate=~
     SkydiveFidgetAnimRateCurrent=~
     SkydiveAdditiveAlpha=~
     SkydiveDriftDelay=~
    ).bIsGliderBack)]
 }
 Transition {
  path=AnimGraph.AnimGraphNode_StateMachine_0.LeanTransitions.AnimStateTransitionNode_0.Transition
  nodes [Result(bCanEnterTransition=Break Fort Anim Input Skydiving(FortAnimInput_Skydiving=Get SkydivingState()) -> (
     ParaGilderRootOffset=~
     ParaGliderRootRotationOverride=~
     bUseParaGlideRootModifier=~
     bIsSkydivingFromLaunchPad=~
     bIsSkydivingFromBus=~
     bIsInVortex=~
     bIsDiving=~
     bIsParachuteOpen=~
     bIsSkydiving=~
     bIsParachuteLeaning=~
     bIsSkydiveLeaning=~
     bIsLeaning=~
     bIsSkydiveDiveMode=~
     bPlaySkydiveDrift=~
     bSkydiveDriftDelayActive=~
     bSkydiveDriftAnimAllowed=~
     bIsGliderRight=~
     bIsGliderCenter=~
     bIsGliderLeft=~
     bIsGliderForward=~
     bIsGliderBack=~
     bHasDeployedAndIsStandingUpright=~
     LocalAccelForward=~
     LocalAccelRight=~
     LocalVelocityRight=~
     SkydiveAimPitch=~
     SkydiveAimYaw=~
     DeployChuteAnimRate=~
     SkydiveDriftAnimRate=~
     SkydiveDriftAnimRateCurrent=~
     SkydiveFidgetAnimRate=~
     SkydiveFidgetAnimRateCurrent=~
     SkydiveAdditiveAlpha=~
     SkydiveDriftDelay=~
    ).bIsLeaning)]
 }
 Transition {path=AnimGraph.AnimGraphNode_StateMachine_0.LeanTransitions.AnimStateTransitionNode_1.Transition, nodes [Result()]}
 Transition {path=AnimGraph.AnimGraphNode_StateMachine_0.LeanTransitions.AnimStateTransitionNode_18.Transition, nodes [Result()]}
 Transition {
  path=AnimGraph.AnimGraphNode_StateMachine_0.LeanTransitions.AnimStateTransitionNode_19.Transition
  nodes [Result(bCanEnterTransition=Get bIsAccelerating2D())]
 }
 Transition {
  path=AnimGraph.AnimGraphNode_StateMachine_0.LeanTransitions.AnimStateTransitionNode_2.Transition
  nodes [Result(bCanEnterTransition=NOT Boolean(A=Get bIsAccelerating2D()))]
 }
 Transition {
  path=AnimGraph.AnimGraphNode_StateMachine_0.LeanTransitions.AnimStateTransitionNode_3.Transition
  nodes [Result(bCanEnterTransition=NOT Boolean(A=Get bIsAccelerating2D()))]
 }
 Transition {
  path=AnimGraph.AnimGraphNode_StateMachine_0.LeanTransitions.AnimStateTransitionNode_4.Transition
  nodes [Result(bCanEnterTransition=Get bIsAccelerating2D())]
 }
 Transition {
  path=AnimGraph.AnimGraphNode_StateMachine_0.LeanTransitions.AnimStateTransitionNode_5.Transition
  nodes [Result(bCanEnterTransition=Get CanLeanIn() -> (CanLeanIn:bool))]
 }
 Locomotion_Frontend {
  entry=Closed
  path=AnimGraph.AnimGraphNode_StateMachine_1.Locomotion_Frontend
  ids=id
  nodes [
   AnimStateNode_0 = Idle_Open() {is=title AnimStateNode}
   AnimStateNode_2 = Closed() {is=title AnimStateNode, stateEntered {guid=#7, notifyName=ClosedEntered}}
   AnimStateNode_3 = Open_Into() {is=title AnimStateNode, stateEntered {guid=#8, notifyName=DeployEntered}}
   AnimStateTransitionNode_0 = AnimStateTransitionNode() {bAutomaticRuleBasedOnSequencePlayerInState=true, blendMode=Linear, crossfadeDuration=0}
   AnimStateTransitionNode_1 = AnimStateTransitionNode() {crossfadeDuration=0}
   AnimStateTransitionNode_2 = AnimStateTransitionNode() {crossfadeDuration=0}
   AnimStateTransitionNode_3 = AnimStateTransitionNode() {crossfadeDuration=0, priorityOrder=2}
  ]
  links [
   {
    dst=Open_Into
    rule=AnimGraph.AnimGraphNode_StateMachine_1.Locomotion_Frontend.AnimStateTransitionNode_1.Transition
    src=Closed
    via=AnimStateTransitionNode_1
   }
   {
    dst=Closed
    rule=AnimGraph.AnimGraphNode_StateMachine_1.Locomotion_Frontend.AnimStateTransitionNode_2.Transition
    src=Idle_Open
    via=AnimStateTransitionNode_2
   }
   {
    dst=Closed
    rule=AnimGraph.AnimGraphNode_StateMachine_1.Locomotion_Frontend.AnimStateTransitionNode_3.Transition
    src=Open_Into
    via=AnimStateTransitionNode_3
   }
   {
    dst=Idle_Open
    rule=AnimGraph.AnimGraphNode_StateMachine_1.Locomotion_Frontend.AnimStateTransitionNode_0.Transition
    src=Open_Into
    via=AnimStateTransitionNode_0
   }
  ]
 }
 Idle_Open {
  path=AnimGraph.AnimGraphNode_StateMachine_1.Locomotion_Frontend.AnimStateNode_0.Idle_Open
  nodes [Output Animation Pose(Result=Sequence Player 'GoggleBud_Glide_FrontEnd_GLIDER'() {sequence=&/GoggleBud_Glide_FrontEnd_GLIDER})]
 }
 Closed {
  path=AnimGraph.AnimGraphNode_StateMachine_1.Locomotion_Frontend.AnimStateNode_2.Closed
  nodes [Output Animation Pose(Result=Sequence Player 'GoggleBud_Closed_FrontEnd_GLIDER'() {bLoopAnimation=false, sequence=&/GoggleBud_Closed_FrontEnd_GLIDER})]
 }
 Open_Into {
  path=AnimGraph.AnimGraphNode_StateMachine_1.Locomotion_Frontend.AnimStateNode_3.Open_Into
  nodes [Output Animation Pose(Result=Sequence Player 'GoggleBud_Into_FrontEnd_GLIDER'() {bLoopAnimation=false, sequence=&/GoggleBud_Into_FrontEnd_GLIDER})]
 }
 Transition {
  path=AnimGraph.AnimGraphNode_StateMachine_1.Locomotion_Frontend.AnimStateTransitionNode_0.Transition
  nodes [
   Result()
   "Equal (Float)"!(A:double="Get Relevant Anim Time Remaining (Open_Into)"!() -> (ReturnValue:float) {is=title K2Node_AnimGetter}, B:double=0) -> (ReturnValue:bool) {is=title K2Node_CallFunction}
  ]
 }
 Transition {
  path=AnimGraph.AnimGraphNode_StateMachine_1.Locomotion_Frontend.AnimStateTransitionNode_1.Transition
  nodes [
   n1 = Get SkydivingState()
   Result(bCanEnterTransition=OR Boolean(
     A=Break Fort Anim Input Skydiving(FortAnimInput_Skydiving=@n1) -> (
      ParaGilderRootOffset=~
      ParaGliderRootRotationOverride=~
      bUseParaGlideRootModifier=~
      bIsSkydivingFromLaunchPad=~
      bIsSkydivingFromBus=~
      bIsInVortex=~
      bIsDiving=~
      bIsParachuteOpen=~
      bIsSkydiving=~
      bIsParachuteLeaning=~
      bIsSkydiveLeaning=~
      bIsLeaning=~
      bIsSkydiveDiveMode=~
      bPlaySkydiveDrift=~
      bSkydiveDriftDelayActive=~
      bSkydiveDriftAnimAllowed=~
      bIsGliderRight=~
      bIsGliderCenter=~
      bIsGliderLeft=~
      bIsGliderForward=~
      bIsGliderBack=~
      bHasDeployedAndIsStandingUpright=~
      LocalAccelForward=~
      LocalAccelRight=~
      LocalVelocityRight=~
      SkydiveAimPitch=~
      SkydiveAimYaw=~
      DeployChuteAnimRate=~
      SkydiveDriftAnimRate=~
      SkydiveDriftAnimRateCurrent=~
      SkydiveFidgetAnimRate=~
      SkydiveFidgetAnimRateCurrent=~
      SkydiveAdditiveAlpha=~
      SkydiveDriftDelay=~
     ).bIsParachuteOpen
     B=Get bForceOpen() -> (bForceOpen:bool)
    ))
  ]
 }
 Transition {
  path=AnimGraph.AnimGraphNode_StateMachine_1.Locomotion_Frontend.AnimStateTransitionNode_2.Transition
  nodes [
   n1 = Get bForceOpen()
   n2 = Break Fort Anim Input Skydiving(FortAnimInput_Skydiving=Get SkydivingState()) -> (
    ParaGilderRootOffset=~
    ParaGliderRootRotationOverride=~
    bUseParaGlideRootModifier=~
    bIsSkydivingFromLaunchPad=~
    bIsSkydivingFromBus=~
    bIsInVortex=~
    bIsDiving=~
    bIsParachuteOpen=~
    bIsSkydiving=~
    bIsParachuteLeaning=~
    bIsSkydiveLeaning=~
    bIsLeaning=~
    bIsSkydiveDiveMode=~
    bPlaySkydiveDrift=~
    bSkydiveDriftDelayActive=~
    bSkydiveDriftAnimAllowed=~
    bIsGliderRight=~
    bIsGliderCenter=~
    bIsGliderLeft=~
    bIsGliderForward=~
    bIsGliderBack=~
    bHasDeployedAndIsStandingUpright=~
    LocalAccelForward=~
    LocalAccelRight=~
    LocalVelocityRight=~
    SkydiveAimPitch=~
    SkydiveAimYaw=~
    DeployChuteAnimRate=~
    SkydiveDriftAnimRate=~
    SkydiveDriftAnimRateCurrent=~
    SkydiveFidgetAnimRate=~
    SkydiveFidgetAnimRateCurrent=~
    SkydiveAdditiveAlpha=~
    SkydiveDriftDelay=~
   )
   Result(bCanEnterTransition=AND Boolean(A=NOT Boolean(A=@n2.bIsParachuteOpen), B=NOT Boolean(A=@n1)))
  ]
 }
 Transition {
  path=AnimGraph.AnimGraphNode_StateMachine_1.Locomotion_Frontend.AnimStateTransitionNode_3.Transition
  nodes [
   n1 = Get bForceOpen()
   n2 = Break Fort Anim Input Skydiving(FortAnimInput_Skydiving=Get SkydivingState()) -> (
    ParaGilderRootOffset=~
    ParaGliderRootRotationOverride=~
    bUseParaGlideRootModifier=~
    bIsSkydivingFromLaunchPad=~
    bIsSkydivingFromBus=~
    bIsInVortex=~
    bIsDiving=~
    bIsParachuteOpen=~
    bIsSkydiving=~
    bIsParachuteLeaning=~
    bIsSkydiveLeaning=~
    bIsLeaning=~
    bIsSkydiveDiveMode=~
    bPlaySkydiveDrift=~
    bSkydiveDriftDelayActive=~
    bSkydiveDriftAnimAllowed=~
    bIsGliderRight=~
    bIsGliderCenter=~
    bIsGliderLeft=~
    bIsGliderForward=~
    bIsGliderBack=~
    bHasDeployedAndIsStandingUpright=~
    LocalAccelForward=~
    LocalAccelRight=~
    LocalVelocityRight=~
    SkydiveAimPitch=~
    SkydiveAimYaw=~
    DeployChuteAnimRate=~
    SkydiveDriftAnimRate=~
    SkydiveDriftAnimRateCurrent=~
    SkydiveFidgetAnimRate=~
    SkydiveFidgetAnimRateCurrent=~
    SkydiveAdditiveAlpha=~
    SkydiveDriftDelay=~
   )
   Result(bCanEnterTransition=AND Boolean(A=NOT Boolean(A=@n2.bIsParachuteOpen), B=NOT Boolean(A=@n1)))
  ]
 }
 Locomotion {
  entry=Closed
  path=AnimGraph.AnimGraphNode_StateMachine_2.Locomotion
  ids=id
  nodes [
   AnimStateNode_0 = Idle_Open()
   AnimStateNode_2 = Closed() {stateEntered {guid=#7, notifyName=ClosedEntered}}
   AnimStateNode_3 = Open_Into() {stateEntered {guid=#8, notifyName=DeployEntered}}
   AnimStateTransitionNode_0 = AnimStateTransitionNode() {bAutomaticRuleBasedOnSequencePlayerInState=true, blendMode=Linear, crossfadeDuration=0}
   AnimStateTransitionNode_1 = AnimStateTransitionNode() {crossfadeDuration=0}
   AnimStateTransitionNode_2 = AnimStateTransitionNode() {crossfadeDuration=0}
   AnimStateTransitionNode_3 = AnimStateTransitionNode() {crossfadeDuration=0, priorityOrder=2}
  ]
  links [
   {
    dst=Open_Into
    rule=AnimGraph.AnimGraphNode_StateMachine_2.Locomotion.AnimStateTransitionNode_1.Transition
    src=Closed
    via=AnimStateTransitionNode_1
   }
   {
    dst=Closed
    rule=AnimGraph.AnimGraphNode_StateMachine_2.Locomotion.AnimStateTransitionNode_2.Transition
    src=Idle_Open
    via=AnimStateTransitionNode_2
   }
   {
    dst=Closed
    rule=AnimGraph.AnimGraphNode_StateMachine_2.Locomotion.AnimStateTransitionNode_3.Transition
    src=Open_Into
    via=AnimStateTransitionNode_3
   }
   {
    dst=Idle_Open
    rule=AnimGraph.AnimGraphNode_StateMachine_2.Locomotion.AnimStateTransitionNode_0.Transition
    src=Open_Into
    via=AnimStateTransitionNode_0
   }
  ]
 }
 Idle_Open {
  path=AnimGraph.AnimGraphNode_StateMachine_2.Locomotion.AnimStateNode_0.Idle_Open
  nodes [Output Animation Pose(Result=Sequence Player 'GoggleBud_Glide_GLIDER'() {sequence=&/GoggleBud_Glide_GLIDER})]
 }
 Closed {
  path=AnimGraph.AnimGraphNode_StateMachine_2.Locomotion.AnimStateNode_2.Closed
  nodes [Output Animation Pose(Result=Sequence Player 'GoggleBud_Closed_GLIDER'() {sequence=&/GoggleBud_Closed_GLIDER})]
 }
 Open_Into {
  path=AnimGraph.AnimGraphNode_StateMachine_2.Locomotion.AnimStateNode_3.Open_Into
  nodes [
   n1 = Break Fort Anim Input Skydiving(FortAnimInput_Skydiving=Get SkydivingState()) -> (
    ParaGilderRootOffset=~
    ParaGliderRootRotationOverride=~
    bUseParaGlideRootModifier=~
    bIsSkydivingFromLaunchPad=~
    bIsSkydivingFromBus=~
    bIsInVortex=~
    bIsDiving=~
    bIsParachuteOpen=~
    bIsSkydiving=~
    bIsParachuteLeaning=~
    bIsSkydiveLeaning=~
    bIsLeaning=~
    bIsSkydiveDiveMode=~
    bPlaySkydiveDrift=~
    bSkydiveDriftDelayActive=~
    bSkydiveDriftAnimAllowed=~
    bIsGliderRight=~
    bIsGliderCenter=~
    bIsGliderLeft=~
    bIsGliderForward=~
    bIsGliderBack=~
    bHasDeployedAndIsStandingUpright=~
    LocalAccelForward=~
    LocalAccelRight=~
    LocalVelocityRight=~
    SkydiveAimPitch=~
    SkydiveAimYaw=~
    DeployChuteAnimRate=~
    SkydiveDriftAnimRate=~
    SkydiveDriftAnimRateCurrent=~
    SkydiveFidgetAnimRate=~
    SkydiveFidgetAnimRateCurrent=~
    SkydiveAdditiveAlpha=~
    SkydiveDriftDelay=~
   )
   Output Animation Pose(Result=Blendspace Player 'BS_GoggleBud_Into_GLIDER'(X=@n1.SkydiveAimYaw, Y=@n1.SkydiveAimPitch) {bLoop=false, blendSpace=&/BS_GoggleBud_Into_GLIDER})
  ]
 }
 Transition {
  path=AnimGraph.AnimGraphNode_StateMachine_2.Locomotion.AnimStateTransitionNode_0.Transition
  nodes [
   Result()
   "Equal (Float)"(A="Get Relevant Anim Time Remaining (Open_Into)"(), B=0)
  ]
 }
 Transition {
  path=AnimGraph.AnimGraphNode_StateMachine_2.Locomotion.AnimStateTransitionNode_1.Transition
  nodes [
   n1 = Get SkydivingState()
   Result(bCanEnterTransition=OR Boolean(
     A=Break Fort Anim Input Skydiving(FortAnimInput_Skydiving=@n1) -> (
      ParaGilderRootOffset=~
      ParaGliderRootRotationOverride=~
      bUseParaGlideRootModifier=~
      bIsSkydivingFromLaunchPad=~
      bIsSkydivingFromBus=~
      bIsInVortex=~
      bIsDiving=~
      bIsParachuteOpen=~
      bIsSkydiving=~
      bIsParachuteLeaning=~
      bIsSkydiveLeaning=~
      bIsLeaning=~
      bIsSkydiveDiveMode=~
      bPlaySkydiveDrift=~
      bSkydiveDriftDelayActive=~
      bSkydiveDriftAnimAllowed=~
      bIsGliderRight=~
      bIsGliderCenter=~
      bIsGliderLeft=~
      bIsGliderForward=~
      bIsGliderBack=~
      bHasDeployedAndIsStandingUpright=~
      LocalAccelForward=~
      LocalAccelRight=~
      LocalVelocityRight=~
      SkydiveAimPitch=~
      SkydiveAimYaw=~
      DeployChuteAnimRate=~
      SkydiveDriftAnimRate=~
      SkydiveDriftAnimRateCurrent=~
      SkydiveFidgetAnimRate=~
      SkydiveFidgetAnimRateCurrent=~
      SkydiveAdditiveAlpha=~
      SkydiveDriftDelay=~
     ).bIsParachuteOpen
     B=Get bForceOpen()
    ))
  ]
 }
 Transition {
  path=AnimGraph.AnimGraphNode_StateMachine_2.Locomotion.AnimStateTransitionNode_2.Transition
  nodes [
   n1 = Get bForceOpen()
   n2 = Break Fort Anim Input Skydiving(FortAnimInput_Skydiving=Get SkydivingState()) -> (
    ParaGilderRootOffset=~
    ParaGliderRootRotationOverride=~
    bUseParaGlideRootModifier=~
    bIsSkydivingFromLaunchPad=~
    bIsSkydivingFromBus=~
    bIsInVortex=~
    bIsDiving=~
    bIsParachuteOpen=~
    bIsSkydiving=~
    bIsParachuteLeaning=~
    bIsSkydiveLeaning=~
    bIsLeaning=~
    bIsSkydiveDiveMode=~
    bPlaySkydiveDrift=~
    bSkydiveDriftDelayActive=~
    bSkydiveDriftAnimAllowed=~
    bIsGliderRight=~
    bIsGliderCenter=~
    bIsGliderLeft=~
    bIsGliderForward=~
    bIsGliderBack=~
    bHasDeployedAndIsStandingUpright=~
    LocalAccelForward=~
    LocalAccelRight=~
    LocalVelocityRight=~
    SkydiveAimPitch=~
    SkydiveAimYaw=~
    DeployChuteAnimRate=~
    SkydiveDriftAnimRate=~
    SkydiveDriftAnimRateCurrent=~
    SkydiveFidgetAnimRate=~
    SkydiveFidgetAnimRateCurrent=~
    SkydiveAdditiveAlpha=~
    SkydiveDriftDelay=~
   )
   Result(bCanEnterTransition=AND Boolean(A=NOT Boolean(A=@n2.bIsParachuteOpen), B=NOT Boolean(A=@n1)))
  ]
 }
 Transition {
  path=AnimGraph.AnimGraphNode_StateMachine_2.Locomotion.AnimStateTransitionNode_3.Transition
  nodes [
   n1 = Get bForceOpen()
   n2 = Break Fort Anim Input Skydiving(FortAnimInput_Skydiving=Get SkydivingState()) -> (
    ParaGilderRootOffset=~
    ParaGliderRootRotationOverride=~
    bUseParaGlideRootModifier=~
    bIsSkydivingFromLaunchPad=~
    bIsSkydivingFromBus=~
    bIsInVortex=~
    bIsDiving=~
    bIsParachuteOpen=~
    bIsSkydiving=~
    bIsParachuteLeaning=~
    bIsSkydiveLeaning=~
    bIsLeaning=~
    bIsSkydiveDiveMode=~
    bPlaySkydiveDrift=~
    bSkydiveDriftDelayActive=~
    bSkydiveDriftAnimAllowed=~
    bIsGliderRight=~
    bIsGliderCenter=~
    bIsGliderLeft=~
    bIsGliderForward=~
    bIsGliderBack=~
    bHasDeployedAndIsStandingUpright=~
    LocalAccelForward=~
    LocalAccelRight=~
    LocalVelocityRight=~
    SkydiveAimPitch=~
    SkydiveAimYaw=~
    DeployChuteAnimRate=~
    SkydiveDriftAnimRate=~
    SkydiveDriftAnimRateCurrent=~
    SkydiveFidgetAnimRate=~
    SkydiveFidgetAnimRateCurrent=~
    SkydiveAdditiveAlpha=~
    SkydiveDriftDelay=~
   )
   Result(bCanEnterTransition=AND Boolean(A=NOT Boolean(A=@n2.bIsParachuteOpen), B=NOT Boolean(A=@n1)))
  ]
 }
 EventGraph {nodes [
   n1 = Event BlueprintUpdateAnimation()
   > n2 = Sequence!(execute) -> (then_0, then_1, then_2:Exec, then_3:Exec, then_4:Exec, then_5:Exec, then_6:Exec, then_7:Exec, then_8:Exec)
   > Set Time(Time:double=@n1.DeltaTimeX) -> (Output_Get:double)
   n3 = float * float(A=float * float(A=Get Time() -> (Time:double), B=15), B=Get Speed() -> (Speed:double)) {is=title K2Node_PromotableOperator}
   n4 = Get ShipOutsidePlate_Tick() -> (ShipOutsidePlate_Tick:double)
   n5 = Set ShipOutsidePlate_Value(execute=@n2.then_4, ShipOutsidePlate_Value:double="% (Float)"!(A:double=float + float(A=@n4, B=@n3), B:double=360) -> (ReturnValue:double)) -> (Output_Get:double)
   > Set ShipOutsidePlate_Rotator(ShipOutsidePlate_Rotator:Rotator=MakeRotator(Roll=@n5)) -> (Output_Get:Rotator)
   > Set ShipOutsidePlate_Tick(ShipOutsidePlate_Tick:double=Get ShipOutsidePlate_Value() -> (ShipOutsidePlate_Value:double)) -> (Output_Get:double)
   n6 = GetCurveValue(CurveName=AlienA_TentaculesRate)
   n7 = Branch(execute=@n2.then_6, Condition="Not Equal (Float)"(A=@n6))
   > Set AlienA_TentaculesRate(AlienA_TentaculesRate:float=float + float(A=@n6, B=0.5)) -> (Output_Get:float)
   Set AlienA_TentaculesRate(execute=@n7.else, AlienA_TentaculesRate=0.5)
   n8 = Branch(
    execute=@n2.then_8
    Condition=IsCosmeticObjectDisplayedInFrontendContext!(CosmeticObject:Object Reference=Self-Reference()) -> (ReturnValue:bool)
   )
   > Set FrontendLevel(FrontendLevel:bool=true) -> (Output_Get:bool)
   Set FrontendLevel(execute=@n8.else, FrontendLevel=false)
   n9 = float * float(A=float * float(A=Get Time(), B=40), B=Get Speed())
   n10 = Get ShipMiddlePlate_Tick() -> (ShipMiddlePlate_Tick:double)
   n11 = Set ShipMiddlePlate_Value(execute=@n2.then_3, ShipMiddlePlate_Value:double="% (Float)"(A=float + float(A=@n10, B=@n9), B=360)) -> (Output_Get:double)
   > Set ShipMiddlePlate_Rotator(ShipMiddlePlate_Rotator:Rotator=MakeRotator(Roll=@n11)) -> (Output_Get:Rotator)
   > Set ShipMiddlePlate_Tick(ShipMiddlePlate_Tick:double=Get ShipMiddlePlate_Value() -> (ShipMiddlePlate_Value:double)) -> (Output_Get:double)
   n12 = GetCurveValue(CurveName=AlienB_TentaculesRate)
   n13 = Branch(execute=@n2.then_7, Condition="Not Equal (Float)"(A=@n12))
   > Set AlienB_TentaculesRate(AlienB_TentaculesRate:float=float + float(A=@n12, B=0.5)) -> (Output_Get:float)
   Set AlienB_TentaculesRate(execute=@n13.else, AlienB_TentaculesRate=0.5)
   n14 = Break Fort Anim Input Skydiving(FortAnimInput_Skydiving=Get SkydivingState()) -> (
    ParaGilderRootOffset=~
    ParaGliderRootRotationOverride=~
    bUseParaGlideRootModifier=~
    bIsSkydivingFromLaunchPad=~
    bIsSkydivingFromBus=~
    bIsInVortex=~
    bIsDiving=~
    bIsParachuteOpen=~
    bIsSkydiving=~
    bIsParachuteLeaning=~
    bIsSkydiveLeaning=~
    bIsLeaning=~
    bIsSkydiveDiveMode=~
    bPlaySkydiveDrift=~
    bSkydiveDriftDelayActive=~
    bSkydiveDriftAnimAllowed=~
    bIsGliderRight=~
    bIsGliderCenter=~
    bIsGliderLeft=~
    bIsGliderForward=~
    bIsGliderBack=~
    bHasDeployedAndIsStandingUpright=~
    LocalAccelForward=~
    LocalAccelRight=~
    LocalVelocityRight=~
    SkydiveAimPitch=~
    SkydiveAimYaw=~
    DeployChuteAnimRate=~
    SkydiveDriftAnimRate=~
    SkydiveDriftAnimRateCurrent=~
    SkydiveFidgetAnimRate=~
    SkydiveFidgetAnimRateCurrent=~
    SkydiveAdditiveAlpha=~
    SkydiveDriftDelay=~
   )
   Branch(
    execute=@n2.then_1
    Condition="Not Equal (Enum)"!(
     A:ESkydivingDirection Enum=@n14.DirectionLast
     B:ESkydivingDirection Enum=Get Direction Last() -> (Direction Last:ESkydivingDirection Enum=Center)
    ) -> (ReturnValue:bool)
   )
   > Set Direction Tick(Direction Tick:ESkydivingDirection Enum=@n14.DirectionLast) -> (Output_Get:ESkydivingDirection Enum=Center)
   > Set CanLeanIn(CanLeanIn:bool=true) -> (Output_Get:bool)
   > Set Direction Last(Direction Last:ESkydivingDirection Enum=@n14.DirectionLast) -> (Output_Get:ESkydivingDirection Enum=Center)
   > Set Direction Tick(Direction Tick=Get Direction Last() -> (Direction Last=Center)) -> (Output_Get=Center)
   > Delay(Duration=0.3)
   > Set CanLeanIn(CanLeanIn=false)
   n15 = ForwardBack_Value_Interp!(
    execute=@n2.then_2
    Time:double=Get Time()
    Currnet_Value:double=Get Speed()
    Idle_Value:double=1
    Idle_InterpSpeed:double=2
    Forward_Value:double=2
    Forward_InterpSpeed:double=2
    Back_Value:double=0.5
    Back_InterpSpeed:double=2
    LocalAccelForward:double=Break Fort Anim Input Skydiving(FortAnimInput_Skydiving=Get SkydivingState()) -> (
     ParaGilderRootOffset=~
     ParaGliderRootRotationOverride=~
     bUseParaGlideRootModifier=~
     bIsSkydivingFromLaunchPad=~
     bIsSkydivingFromBus=~
     bIsInVortex=~
     bIsDiving=~
     bIsParachuteOpen=~
     bIsSkydiving=~
     bIsParachuteLeaning=~
     bIsSkydiveLeaning=~
     bIsLeaning=~
     bIsSkydiveDiveMode=~
     bPlaySkydiveDrift=~
     bSkydiveDriftDelayActive=~
     bSkydiveDriftAnimAllowed=~
     bIsGliderRight=~
     bIsGliderCenter=~
     bIsGliderLeft=~
     bIsGliderForward=~
     bIsGliderBack=~
     bHasDeployedAndIsStandingUpright=~
     LocalAccelForward=~
     LocalAccelRight=~
     LocalVelocityRight=~
     SkydiveAimPitch=~
     SkydiveAimYaw=~
     DeployChuteAnimRate=~
     SkydiveDriftAnimRate=~
     SkydiveDriftAnimRateCurrent=~
     SkydiveFidgetAnimRate=~
     SkydiveFidgetAnimRateCurrent=~
     SkydiveAdditiveAlpha=~
     SkydiveDriftDelay=~
    ).LocalAccelForward
   ) -> (then, New_Value:double)
   > Set Speed(Speed:double=@n15) -> (Output_Get:double)
   n16 = Event BlueprintInitializeAnimation()
   n17 = GetOwningComponent()
   HideBoneByName!(
    execute=[@n16.then, @n18.then]
    self:Skinned Mesh Component Object Reference=@n17
    BoneName:Name=Glider_Base
    PhysBodyOption:EPhysBodyOp Enum=PBO_None
   ) -> (then)
   n18 = AnimNotify_ClosedEntered() {is=title K2Node_Event}
   AnimNotify_DeployEntered() {is=title K2Node_Event}
   > Delay(Duration=0.1)
   > UnHideBoneByName!(execute, self:Skinned Mesh Component Object Reference=@n17, BoneName:Name=Glider_Base) -> (then)
   AnimNotify_TurnONEngines() {is=title K2Node_Event}
   > SetGliderFullyDeployed!(execute, self:Fort Glider Anim Instance Object Reference, bIsFullyDeployed:bool=true) -> (then)
   AnimNotify_IntoEnd() {is=title K2Node_Event}
   > Set IntoEnd(IntoEnd:bool=true) -> (Output_Get:bool)
   n19 = Get OwnerGlider() -> (OwnerGlider:Fort Player Parachute Object Reference)
   Set Actor Hidden In Game!(execute, self:Actor Object Reference=@n19, bNewHidden:bool=true) -> (then) {is=title K2Node_CallFunction}
   Set Actor Hidden In Game(self=@n19, bNewHidden=false)
  ]}
 GetPlayerOpenGliderBool {nodes [
   Inputs()
   > n1 = Branch(Condition=Get IsCosmeticPreview!(self:Fort Player Parachute Object Reference=Get OwnerGlider()) -> (IsCosmeticPreview:bool))
   Outputs!(then:Exec=[@n3.then, @n1.then])
   n2 = Get OwnerFortPawn() -> (OwnerFortPawn:Fort Player Pawn Object Reference)
   n3 = Set bForceOpen(
    execute=@n1.else
    bForceOpen:bool=ShouldForceOpenGlider!(self:Fort Player Animation Interface Object Reference=GetAnimationInterface!(self:Fort Player Pawn Object Reference=@n2) -> (ReturnValue:Fort Player Animation Interface Object Reference)) -> (ReturnValue:bool)
   ) -> (Output_Get:bool)
  ]}
]
variables [
 Acceleration {type=number}
 AlienA_TentaculesRate {type=number}
 AlienB_TentaculesRate {type=number}
 CanLeanIn {type=boolean}
 Direction Last {type=ESkydivingDirection, default=Center}
 Direction Tick {type=ESkydivingDirection, default=Center}
 Example {type=number}
 ForwardBack_Acceleration {type=number}
 FrontendLevel {type=boolean}
 IntoEnd {type=boolean}
 Local Accel Dir {type=ESkydivingDirection, default=Center}
 ShipMiddlePlate_Rotator {type=Rotator}
 ShipMiddlePlate_Tick {type=number}
 ShipMiddlePlate_Value {type=number}
 ShipOutsidePlate_Rotator {type=Rotator}
 ShipOutsidePlate_Tick {type=number}
 ShipOutsidePlate_Value {type=number}
 Side_Acceleration {type=number}
 Speed {type=number}
 Time {type=number}
]
